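# Physics asset of cape colliders for the male medium WW2 pilot body, evidently the collision shapes that a simulated cape rests against.
# It gives one capsule each to the pelvis, two spine bones, neck, head, clavicles, upper arms, thighs and calves, leaving out forearms, hands and feet.
# Constraints chain the bodies along the skeleton with all swing and twist motion free, so none restricts rotation.
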PhysicsAsset /BRCosmetics/Characters/Player/Male/Medium/Bodies/M_MED_ww2_Pilot_01/Meshes/M_MED_WW2_Pilot_CapeColliders1
bodies {
 calf_l [{shapeType=Capsule, length=33.242, radius=9.9116, center=<-19.4843 -0.5441 -0.2294>, rotation=<rot 0 94.0793 -89.4659>}]
 calf_r [{shapeType=Capsule, length=33.5362, radius=9.5971, center=<19.4486 0.3664 0.2406>, rotation=<rot 0 93.6446 -89.3731>}]
 clavicle_l [{shapeType=Capsule, length=10.3261, radius=8.5802, center=<8.2822 -0.2527 -2.0684>, rotation=<rot 90 -0.0049 179.9951>}]
 clavicle_r [{shapeType=Capsule, length=10.3261, radius=8.5802, center=<-7.9705 0.6201 2.6048>, rotation=<rot 90 0.0198 0.0198>}]
 head [{shapeType=Capsule, length=0, radius=18, center=<11.6361 0.0962 0>, rotation=<rot 0 -90 0>}]
 neck_02 [{shapeType=Capsule, length=0.505, radius=10.755, rotation=<rot 0 -90 0>}]
 pelvis [{shapeType=Capsule, length=11.4559, radius=14.3329, center=<2.2747 -0.2766 1.0154>}]
 spine_03 [{shapeType=Capsule, length=10.0169, radius=13.2572, center=<0 0.5583 0>}]
 spine_05 [{shapeType=Capsule, length=15.9732, radius=13.267, center=<-2.4255 0 0>}]
 thigh_l [{shapeType=Capsule, length=19.002, radius=17.5818, center=<-28.3947 -1.7177 0.0001>, rotation=<rot 90 0 0>}]
 thigh_r [{shapeType=Capsule, length=18.6153, radius=17.4315, center=<27.3972 1.7178 0.0025>, rotation=<rot 90 0 0>}]
 upperarm_l [{shapeType=Capsule, length=18.8261, radius=7.5802, center=<12.1824 -0.2526 -0.6865>, rotation=<rot 90 0.005 -179.995>}]
 upperarm_r [{shapeType=Capsule, length=17.1357, radius=7.1339, center=<-12.8316 -0.0074 -0.1319>, rotation=<rot 0 -88.4734 -86.4203>}]
}
constraints [
 {
  bone1Name=spine_03
  bone2Name=pelvis
  swing1LimitDegrees=45
  swing1Motion=Free
  swing2LimitDegrees=45
  swing2Motion=Free
  twistLimitDegrees=45
  twistMotion=Free
 }
 {
  bone1Name=spine_05
  bone2Name=spine_03
  swing1LimitDegrees=45
  swing1Motion=Free
  swing2LimitDegrees=45
  swing2Motion=Free
  twistLimitDegrees=45
  twistMotion=Free
 }
 {
  bone1Name=clavicle_l
  bone2Name=spine_05
  swing1LimitDegrees=45
  swing1Motion=Free
  swing2LimitDegrees=45
  swing2Motion=Free
  twistLimitDegrees=45
  twistMotion=Free
 }
 {
  bone1Name=upperarm_l
  bone2Name=clavicle_l
  swing1LimitDegrees=45
  swing1Motion=Free
  swing2LimitDegrees=45
  swing2Motion=Free
  twistLimitDegrees=45
  twistMotion=Free
 }
 {
  bone1Name=clavicle_r
  bone2Name=spine_05
  swing1LimitDegrees=45
  swing1Motion=Free
  swing2LimitDegrees=45
  swing2Motion=Free
  twistLimitDegrees=45
  twistMotion=Free
 }
 {
  bone1Name=upperarm_r
  bone2Name=clavicle_r
  swing1LimitDegrees=45
  swing1Motion=Free
  swing2LimitDegrees=45
  swing2Motion=Free
  twistLimitDegrees=45
  twistMotion=Free
 }
 {
  bone1Name=thigh_l
  bone2Name=pelvis
  swing1LimitDegrees=45
  swing1Motion=Free
  swing2LimitDegrees=45
  swing2Motion=Free
  twistLimitDegrees=45
  twistMotion=Free
 }
 {
  bone1Name=calf_l
  bone2Name=thigh_l
  swing1LimitDegrees=45
  swing1Motion=Free
  swing2LimitDegrees=45
  swing2Motion=Free
  twistLimitDegrees=45
  twistMotion=Free
 }
 {
  bone1Name=thigh_r
  bone2Name=pelvis
  swing1LimitDegrees=45
  swing1Motion=Free
  swing2LimitDegrees=45
  swing2Motion=Free
  twistLimitDegrees=45
  twistMotion=Free
 }
 {
  bone1Name=calf_r
  bone2Name=thigh_r
  swing1LimitDegrees=45
  swing1Motion=Free
  swing2LimitDegrees=45
  swing2Motion=Free
  twistLimitDegrees=45
  twistMotion=Free
 }
 {
  bone1Name=neck_02
  bone2Name=spine_05
  swing1LimitDegrees=45
  swing1Motion=Free
  swing2LimitDegrees=45
  swing2Motion=Free
  twistLimitDegrees=45
  twistMotion=Free
 }
 {
  bone1Name=head
  bone2Name=neck_02
  swing1LimitDegrees=45
  swing1Motion=Free
  swing2LimitDegrees=45
  swing2Motion=Free
  twistLimitDegrees=45
  twistMotion=Free
 }
]
