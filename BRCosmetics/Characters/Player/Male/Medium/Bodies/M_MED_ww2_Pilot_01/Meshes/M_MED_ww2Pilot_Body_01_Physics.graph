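# Physics asset for the male medium WW2 pilot player body cosmetic, providing simplified collision for the lower body only.
# One large capsule covers the pelvis, and paired, offset capsules approximate each thigh and calf, with no bodies for feet or upper body.
# Constraints chain calves to thighs and thighs to pelvis, but all swing and twist motions are free, so their angle limits do not restrict movement.
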PhysicsAsset /BRCosmetics/Characters/Player/Male/Medium/Bodies/M_MED_ww2_Pilot_01/Meshes/M_MED_ww2Pilot_Body_01_Physics
bodies {
 calf_l [
  {shapeType=Capsule, length=21.25, radius=7, center=<-12.1236 -2.5011 0.65>, rotation=<rot 90 -36.8699 -46.8702>}
  {shapeType=Capsule, length=21.25, radius=7, center=<-6.4094 -3.1902 0.65>, rotation=<rot 90 0 -18.4398>}
 ]
 calf_r [
  {shapeType=Capsule, length=21.25, radius=7, center=<7.1227 1.4232 -1.4204>, rotation=<rot 90 0 -29.6907>}
  {shapeType=Capsule, length=21.25, radius=7, center=<19.6078 0.6266 -1.4204>, rotation=<rot 90 -36.8699 -46.8702>}
 ]
 pelvis [{shapeType=Capsule, length=21, radius=22, center=<24.1425 -0.2224 0>, rotation=<rot 90 0 0>}]
 thigh_l [
  {shapeType=Capsule, length=34.5, radius=7.25, center=<-16.8727 -2.7076 0.6623>, rotation=<rot 90 0 0>}
  {shapeType=Capsule, length=27.25, radius=7.25, center=<-19.4271 4.9254 0.6623>, rotation=<rot 90 -2.5305 -2.5304>}
 ]
 thigh_r [
  {shapeType=Capsule, length=35.25, radius=7.25, center=<17.6477 0.9229 -0.5203>, rotation=<rot 90 0 0>}
  {shapeType=Capsule, length=27.25, radius=7.25, center=<22.3975 -4.7613 -0.5208>, rotation=<rot 90 1.7899 -1.0228>}
 ]
}
constraints [
 {
  bone1Name=thigh_l
  bone2Name=pelvis
  swing1LimitDegrees=45
  swing1Motion=Free
  swing2LimitDegrees=45
  swing2Motion=Free
  twistLimitDegrees=45
  twistMotion=Free
 }
 {
  bone1Name=thigh_r
  bone2Name=pelvis
  swing1LimitDegrees=45
  swing1Motion=Free
  swing2LimitDegrees=45
  swing2Motion=Free
  twistLimitDegrees=45
  twistMotion=Free
 }
 {
  bone1Name=calf_r
  bone2Name=thigh_r
  swing1LimitDegrees=45
  swing1Motion=Free
  swing2LimitDegrees=45
  swing2Motion=Free
  twistLimitDegrees=45
  twistMotion=Free
 }
 {
  bone1Name=calf_l
  bone2Name=thigh_l
  swing1LimitDegrees=45
  swing1Motion=Free
  swing2LimitDegrees=45
  swing2Motion=Free
  twistLimitDegrees=45
  twistMotion=Free
 }
]
